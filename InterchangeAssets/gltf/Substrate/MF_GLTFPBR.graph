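# Material function that builds a glTF PBR surface in Substrate, outputting Front Material, Ambient Occlusion and Refraction.
# Dielectric and metal slabs are parameter-blended horizontally, then layered under a weighted clearcoat slab.
# Inputs cover iridescence via thin film, sheen via fuzz, transmission via mean free path, alpha modes and anisotropy with a custom HLSL rotated tangent.
# Refraction has a separate path-tracing branch, and the dispersion input is unconnected and marked unsupported.
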MaterialFunction /InterchangeAssets/gltf/Substrate/MF_GLTFPBR
graph {
 properties {bExposeToLibrary=true}
 nodes [
  FunctionInput() {
   bUsePreviewValueAsDefault=true
   description=unsupported
   inputName=dispersion
   inputType=FunctionInput_Scalar
   sortPriority=29
   previewValue=<0 0 0 0>
  }
  n1 = Constant3Vector() {constant=<rgba 1 1 1 0>}
  n2 = SubstrateWeight!(
   A=SubstrateSlabBSDF!(
    Diffuse Albedo=Constant3Vector()
    F0
    F90
    Roughness=NamedRerouteUsage()
    Anisotropy
    Normal=NamedRerouteUsage()
    Tangent
    SSS MFP=SubstrateTransmittanceToMFP!(TransmittanceColor=@n1, Thickness).MFP
    SSS MFP Scale
    SSS Phase Anisotropy
    Emissive Color
    Second Roughness
    Second Roughness Weight
    Fuzz Roughness
    Fuzz Amount
    Fuzz Color
    "Glint Density (Disabled)"
    "Glint UVs (Disabled)"
   )
   Weight=NamedRerouteUsage()
  )
  n3 = NamedRerouteUsage()
  n4 = Constant3Vector()
  n5 = NamedRerouteUsage()
  n6 = NamedRerouteUsage()
  n7 = NamedRerouteUsage()
  n8 = SubstrateThinFilm!(Normal=NamedRerouteUsage(), F0=NamedRerouteUsage(), F90=NamedRerouteUsage(), Thickness=@n7, IOR=NamedRerouteUsage())
  n9 = NamedRerouteUsage()
  n10 = If(A=@n7, "A > B"=@n8.Specular Color, "A == B"=@n9, "A < B"=@n9)
  n11 = NamedRerouteUsage()
  n12 = NamedRerouteUsage()
  n13 = NamedRerouteUsage()
  n14 = If(A=@n7, "A > B"=@n8.Edge Specular Color, "A == B"=@n13, "A < B"=@n13)
  n15 = Multiply(A=SquareRoot(None=NamedRerouteUsage()), B=NamedRerouteUsage())
  n16 = SubstrateMetalnessToDiffuseAlbedoF0!(BaseColor=NamedRerouteUsage(), Metallic=NamedRerouteUsage(), Specular=NamedRerouteUsage())
  n17 = NamedRerouteDeclaration(None=@n16.F0) {name=MetalF0, nodeColor=<rgba 0.7812 1 0 1>}
  n18 = NamedRerouteUsage()
  n19 = SubstrateThinFilm(Normal=NamedRerouteUsage(), F0=NamedRerouteUsage(), Thickness=NamedRerouteUsage(), IOR=NamedRerouteUsage())
  n20 = NamedRerouteUsage()
  n21 = If(A=@n18, "A > B"=@n19.Specular Color, "A == B"=@n20, "A < B"=@n20)
  n22 = NamedRerouteUsage()
  n23 = Constant3Vector() {constant=<rgba 1 1 1 0>}
  n24 = Constant3Vector() {constant=<rgba 1 1 1 0>}
  n25 = If(A=@n18, "A > B"=@n19.Edge Specular Color, "A == B"=@n24, "A < B"=@n24)
  n26 = NamedRerouteUsage()
  n27 = SubstrateHorizontalMixing!(
   Background=SubstrateSlabBSDF(
    Diffuse Albedo=LinearInterpolate(A=@n3, B=@n4, Alpha=@n5)
    F0=LinearInterpolate(A=@n6, B=@n10, Alpha=@n11)
    F90=LinearInterpolate(A=@n12, B=@n14, Alpha=@n11)
    Roughness=NamedRerouteUsage()
    Anisotropy=NamedRerouteUsage()
    Normal=NamedRerouteUsage()
    Tangent=NamedRerouteUsage()
    SSS MFP=SubstrateTransmittanceToMFP(TransmittanceColor=@n15).MFP
    Emissive Color=NamedRerouteUsage()
    Fuzz Roughness=NamedRerouteUsage()
    Fuzz Amount=NamedRerouteUsage()
    Fuzz Color=NamedRerouteUsage()
   ) {subSurfaceType=MSS_SimpleVolume}
   Foreground=SubstrateSlabBSDF(
    Diffuse Albedo=@n16.DiffuseAlbedo
    F0=LinearInterpolate(A=@n17, B=@n21, Alpha=@n22)
    F90=LinearInterpolate(A=@n23, B=@n25, Alpha=@n26)
    Roughness=NamedRerouteUsage()
    Anisotropy=NamedRerouteUsage()
    Normal=NamedRerouteUsage()
    Tangent=NamedRerouteUsage()
    Emissive Color=NamedRerouteUsage()
   )
   Mix=NamedRerouteUsage()
  ) {bUseParameterBlending=true}
  FunctionOutput(None=SubstrateWeight(A=SubstrateVerticalLayering!(Top=@n2, Bottom=@n27, Top Thickness), Weight=NamedRerouteUsage())) {outputName=Front Material}
  n28 = FunctionInput() {bUsePreviewValueAsDefault=true, inputName=occlusion, inputType=FunctionInput_Scalar, sortPriority=5, previewValue=<1 1 1 1>}
  FunctionOutput(None=@n28) {outputName=Ambient Occlusion}
  n29 = Divide(B=NamedRerouteUsage())
  n30 = Fresnel(BaseReflectFractionIn=NamedRerouteUsage())
  FunctionOutput(None=PathTracingQualitySwitch!(Normal=LinearInterpolate(B=@n29, Alpha=@n30) {constA=1}, PathTraced=NamedRerouteUsage())) {outputName=Refraction}
  NamedRerouteDeclaration(None=FunctionInput() {bUsePreviewValueAsDefault=true, inputName=base_color, previewValue=<1 1 1 1>}) {name=BaseColor, nodeColor=<rgba 1 0 0.9375 1>}
  n31 = FunctionInput() {bUsePreviewValueAsDefault=true, inputName=roughness, inputType=FunctionInput_Scalar, sortPriority=2, previewValue=<1 1 1 1>}
  NamedRerouteDeclaration(None=@n31) {name=Roughness, nodeColor=<rgba 1 0.4922 0 1>}
  n32 = FunctionInput() {bUsePreviewValueAsDefault=true, inputName=iridescence, inputType=FunctionInput_Scalar, sortPriority=13, previewValue=<0 0 0 0>}
  NamedRerouteDeclaration(None=@n32) {name=ThinFilm, nodeColor=<rgba 0.096 0.156 0.9297 1>}
  n33 = FunctionInput() {bUsePreviewValueAsDefault=true, inputName=iridescence_ior, inputType=FunctionInput_Scalar, sortPriority=14, previewValue=<1.3 1.3 1.3 1.3>}
  NamedRerouteDeclaration(None=@n33) {name=ThinFilmIOR, nodeColor=<rgba 0.112 0.0894 0.544 1>}
  n34 = FunctionInput() {
   bUsePreviewValueAsDefault=true
   inputName=iridescence_thickness
   inputType=FunctionInput_Scalar
   sortPriority=15
   previewValue=<100 100 100 100>
  }
  NamedRerouteDeclaration(None=Multiply(A=@n34, B=Custom() {code="return PI * 0.0001\r\n;", description="Pi Factor ", outputType=CMOT_Float1, inputs []})) {name=ThinFilmThickness, nodeColor=<rgba 0.2891 0 1 1>}
  MaterialExpressionFunctionInput_16 = FunctionInput() {bUsePreviewValueAsDefault=true, inputName=sheen_color, sortPriority=16, previewValue=<0 0 0 0>}
  NamedRerouteDeclaration(None=Divide(A=@MaterialExpressionFunctionInput_16, B=NamedRerouteUsage())) {name=SheenColor, nodeColor=<rgba 0.5938 0 1 1>}
  n35 = FunctionInput() {bUsePreviewValueAsDefault=true, inputName=sheen_roughness, inputType=FunctionInput_Scalar, sortPriority=17, previewValue=<0 0 0 0>}
  NamedRerouteDeclaration(None=@n35) {name=SheenRoughness, nodeColor=<rgba 0 1 0.2734 1>}
  n36 = FunctionInput() {bUsePreviewValueAsDefault=true, inputName=clearcoat, inputType=FunctionInput_Scalar, sortPriority=18, previewValue=<0 0 0 0>}
  NamedRerouteDeclaration(None=@n36) {name=ClearCoat, nodeColor=<rgba 1 0.1875 0 1>}
  n37 = FunctionInput() {bUsePreviewValueAsDefault=true, inputName=clearcoat_roughness, inputType=FunctionInput_Scalar, sortPriority=19, previewValue=<0 0 0 0>}
  NamedRerouteDeclaration(None=@n37) {name=ClearCoatRoughness, nodeColor=<rgba 0 0.1328 1 1>}
  MaterialExpressionVertexTangentWS_2 = VertexTangentWS()
  n38 = Custom!(SurfaceTangent=@MaterialExpressionVertexTangentWS_2) {
   code="{\r\n##if MATERIAL_TANGENTSPACENORMAL\r\n\tconst float3 T = float3(1,0,0);\r\n##else\r\n\tconst float3 T = normalize(SurfaceTangent);\r\n##endif\r\n\treturn T;\r\n}"
   description=GetMtlDefTangentFrame
   inputs [{inputName=SurfaceTangent, input {expression=&self:MaterialExpressionVertexTangentWS_2}}]
  }
  NamedRerouteDeclaration(None=FunctionInput(Preview=@n38) {bUsePreviewValueAsDefault=true, inputName=tangent, sortPriority=4, previewValue=<1 1 1 1>}) {name=Tangent, nodeColor=<rgba 1 0.6328 0 1>}
  n39 = FunctionInput() {bUsePreviewValueAsDefault=true, inputName=anisotropy_strength, inputType=FunctionInput_Scalar, sortPriority=26, previewValue=<0 0 0 0>}
  NamedRerouteDeclaration(None=@n39) {name=Anisotropy, nodeColor=<rgba 0.5469 1 0 1>}
  n40 = FunctionInput() {bUsePreviewValueAsDefault=true, inputName=anisotropy_rotation, inputType=FunctionInput_Scalar, sortPriority=27, previewValue=<0 0 0 0>}
  MaterialExpressionNamedRerouteDeclaration_23 = NamedRerouteDeclaration(None=@n40) {name=AnisotropyRotation, nodeColor=<rgba 1 0 0.8672 1>}
  MaterialExpressionNamedRerouteUsage_2 = NamedRerouteUsage()
  MaterialExpressionNamedRerouteUsage_3 = NamedRerouteUsage()
  MaterialExpressionFunctionInput_29 = FunctionInput() {bUsePreviewValueAsDefault=true, inputName=anisotropy_direction, inputType=FunctionInput_Vector2, sortPriority=27, previewValue=<1 0 0 0>}
  n41 = Custom!(
   Rotation=@MaterialExpressionNamedRerouteDeclaration_23
   SurfaceTangent=@MaterialExpressionNamedRerouteUsage_2
   SurfaceNormal=@MaterialExpressionNamedRerouteUsage_3
   Direction=@MaterialExpressionFunctionInput_29
  ) {
   code="const float3 N = SurfaceNormal;\r\nconst float3 T = SurfaceTangent;\r\nconst float3 B = normalize(cross(SurfaceTangent, SurfaceNormal));\r\n\r\nconst float cosAngle = cos(Rotation);\r\nconst float sinAngle = sin(Rotation);\r\n\r\nfloat3 Dir = float3(mul(float2x2(cosAngle, sinAngle, -sinAngle, cosAngle), normalize(Direction)), 0.);\r\n\r\nconst float3 Out = normalize(mul(float3x3(T, B, N), Dir));\r\n\r\nreturn Out;"
   description=GetAnisotropyTangent
   inputs [
    {inputName=Rotation, input {expression=&self:MaterialExpressionNamedRerouteDeclaration_23}}
    {inputName=SurfaceTangent, input {expression=&self:MaterialExpressionNamedRerouteUsage_2}}
    {inputName=SurfaceNormal, input {expression=&self:MaterialExpressionNamedRerouteUsage_3}}
    {inputName=Direction, input {expression=&self:MaterialExpressionFunctionInput_29}}
   ]
  }
  n42 = NamedRerouteUsage()
  NamedRerouteDeclaration(None=If(A=NamedRerouteUsage(), "A > B"=@n41, "A == B"=@n42, "A < B"=@n42)) {name=AnisotropyTangent, nodeColor=<rgba 1 0 0.3047 1>}
  n43 = FunctionInput() {bUsePreviewValueAsDefault=true, inputName=ior, inputType=FunctionInput_Scalar, sortPriority=9, previewValue=<1.5 1.5 1.5 1.5>}
  n44 = NamedRerouteDeclaration(None=MaterialFunctionCall!(IOR=NamedRerouteDeclaration(None=@n43) {name=IOR, nodeColor=<rgba 1 0 0.4219 1>}) {materialFunction=/Engine/Functions/Substrate/Substrate-IOR-To-F0}.F0) {name=IORtoF0, nodeColor=<rgba 0.4297 0 1 1>}
  n45 = Constant() {r=0.08}
  NamedRerouteDeclaration(None=Saturate(None=Divide(A=@n44, B=@n45))) {name=LegacySpec, nodeColor=<rgba 0 0.8828 1 1>}
  n46 = FunctionInput() {
   bUsePreviewValueAsDefault=true
   description="enum: OPAQUE, MASK, BLEND"
   inputName=alpha_mode
   inputType=FunctionInput_Scalar
   sortPriority=11
   previewValue=<0 0 0 1.5>
  }
  n47 = NamedRerouteDeclaration(None=Floor(None=Clamp(None=@n46) {maxDefault=2})) {name=AlphaMode, nodeColor=<rgba 0.2656 1 0 1>}
  n48 = NamedRerouteUsage()
  n49 = FunctionInput() {bUsePreviewValueAsDefault=true, inputName=alpha, inputType=FunctionInput_Scalar, sortPriority=10, previewValue=<1 1 1 1.5>}
  n50 = FunctionInput() {bUsePreviewValueAsDefault=true, inputName=alpha_cutoff, inputType=FunctionInput_Scalar, sortPriority=12, previewValue=<0.5 0.5 0.5 0.5>}
  n51 = Constant() {r=1}
  n52 = If(
   A=@n47
   "A > B"=@n48
   "A == B"=If(
    A=NamedRerouteDeclaration(None=@n49) {name=Alpha, nodeColor=<rgba 0.0547 0 1 1>}
    B=NamedRerouteDeclaration(None=@n50) {name=AlphaCutoff, nodeColor=<rgba 1 0 0.5859 1>}
    "A > B"=@n51
    "A == B"=@n51
    "A < B"=Constant()
   )
   "A < B"=@n48
  ) {constB=1}
  NamedRerouteDeclaration(None=If(A=@n47, "A > B"=@n52, "A == B"=Constant() {r=1}, "A < B"=@n52)) {name=Opacity, nodeColor=<rgba 0 0.8594 1 1>}
  n53 = FunctionInput() {bUsePreviewValueAsDefault=true, inputName=emissive, sortPriority=21, previewValue=<0 0 0 0>}
  n54 = FunctionInput() {bUsePreviewValueAsDefault=true, inputName=emissive_strength, inputType=FunctionInput_Scalar, sortPriority=22, previewValue=<1 1 1 1>}
  NamedRerouteDeclaration(None=Multiply(
    A=NamedRerouteDeclaration(None=@n53) {name=Emissive, nodeColor=<rgba 0 0.8594 1 1>}
    B=NamedRerouteDeclaration(None=@n54) {name=EmissiveStrength, nodeColor=<rgba 1 0.8203 0 1>}
   )) {name=EmissiveColor, nodeColor=<rgba 0.2656 0 1 1>}
  n55 = FunctionInput() {bUsePreviewValueAsDefault=true, inputName=specular, inputType=FunctionInput_Scalar, sortPriority=7, previewValue=<1 1 1 1>}
  n56 = NamedRerouteDeclaration(None=@n55) {name=Specular, nodeColor=<rgba 1 0 0.3281 1>}
  n57 = FunctionInput() {bUsePreviewValueAsDefault=true, inputName=specular_color, sortPriority=8, previewValue=<1 1 1 1>}
  n58 = Multiply(A=NamedRerouteDeclaration(None=@n57) {name=SpecularColor, nodeColor=<rgba 0 0.6484 1 1>}, B=@n44)
  NamedRerouteDeclaration(None=Multiply(A=@n56, B=Min(A=@n58))) {name=SpecF0, nodeColor=<rgba 1 0.7969 0 1>}
  NamedRerouteDeclaration(None=Multiply(A=Constant3Vector() {constant=<rgba 1 1 1 0>}, B=@n56)) {name=SpecF90, nodeColor=<rgba 0.4766 0 1 1>}
  MaterialExpressionVertexNormalWS_3 = VertexNormalWS()
  n59 = Custom!(SurfaceNormal=@MaterialExpressionVertexNormalWS_3) {
   code="{\r\n##if MATERIAL_TANGENTSPACENORMAL\r\n\tconst float3 N = float3(0,0,1);\r\n##else\r\n\tconst float3 N = SurfaceNormal;\r\n##endif\r\n\treturn N;\r\n}"
   description=GetMtlDefNormalFrame
   inputs [{inputName=SurfaceNormal, input {expression=&self:MaterialExpressionVertexNormalWS_3}}]
  }
  NamedRerouteDeclaration(None=FunctionInput(Preview=@n59) {bUsePreviewValueAsDefault=true, inputName=normal, sortPriority=3, previewValue=<1 1 1 1>}) {name=Normal, nodeColor=<rgba 0.3125 0 1 1>}
  n60 = Convert(None=@MaterialExpressionFunctionInput_16) {
   nodeName=Break Float3
   convertInputs [{
     type=Vector3
     expressionInput {
      expression=&self:MaterialExpressionFunctionInput_16
      inputName=None
      mask=0
      maskA=0
      maskB=0
      maskG=0
      maskR=0
      outputIndex=0
     }
    }]
   convertOutputs [{type=Scalar}, {type=Scalar}, {type=Scalar}]
  }
  NamedRerouteDeclaration(None=Max(A=Max(A=@n60, B=@n60), B=@n60)) {name=Sheen, nodeColor=<rgba 0.8047 0 1 1>}
  n61 = FunctionInput() {bUsePreviewValueAsDefault=true, inputName=attenuation_color, sortPriority=25, previewValue=<1 1 1 0>}
  n62 = SquareRoot(None=@n61)
  n63 = FunctionInput() {bUsePreviewValueAsDefault=true, inputName=thickness, inputType=FunctionInput_Scalar, sortPriority=23, previewValue=<0 0 0 0>}
  n64 = FunctionInput() {bUsePreviewValueAsDefault=true, inputName=attenuation_distance, inputType=FunctionInput_Scalar, sortPriority=24, previewValue=<10 0 0 0>}
  n65 = Divide(
   A=NamedRerouteDeclaration(None=@n63) {name=Thickness, nodeColor=<rgba 0.8047 1 0 1>}
   B=NamedRerouteDeclaration(None=@n64) {name=Depth, nodeColor=<rgba 0.8984 1 0 1>}
  )
  NamedRerouteDeclaration(None=Multiply(A=Power(Base=@n62, Exp=@n65), B=NamedRerouteUsage())) {name=TransmissionColor, nodeColor=<rgba 0.5234 0 1 1>}
  n66 = FunctionInput() {bUsePreviewValueAsDefault=true, inputName=metallic, inputType=FunctionInput_Scalar, sortPriority=1, previewValue=<1 1 1 1>}
  NamedRerouteDeclaration(None=@n66) {name=Metal, nodeColor=<rgba 0.0078 1 0 1>}
  MaterialExpressionVertexNormalWS_0 = VertexNormalWS()
  n67 = Custom(SurfaceNormal=@MaterialExpressionVertexNormalWS_0) {
   code="{\r\n##if MATERIAL_TANGENTSPACENORMAL\r\n\tconst float3 N = float3(0,0,1);\r\n##else\r\n\tconst float3 N = SurfaceNormal;\r\n##endif\r\n\treturn N;\r\n}"
   description=GetMtlDefNormalFrame
   inputs [{inputName=SurfaceNormal, input {expression=&self:MaterialExpressionVertexNormalWS_0}}]
  }
  n68 = FunctionInput(Preview=@n67) {bUsePreviewValueAsDefault=true, inputName=clearcoat_normal, sortPriority=20, previewValue=<0 0 0 0>}
  NamedRerouteDeclaration(None=@n68) {name=ClearCoatNormal, nodeColor=<rgba 0 1 0.7891 1>}
  n69 = FunctionInput() {bUsePreviewValueAsDefault=true, inputName=transmission, inputType=FunctionInput_Scalar, sortPriority=6, previewValue=<0 0 0 0>}
  NamedRerouteDeclaration(None=@n69) {name=Transmission, nodeColor=<rgba 0.7813 0 1 1>}
 ]
}
