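# Substrate master material for glTF assets in Interchange, covering core PBR plus clearcoat, sheen, iridescence, anisotropy, transmission, volume and specular.
# Each texture gets transformed UVs and explicit-derivative sampling, combines with its factors, and uses static switches for optional inputs.
# A static switch selects metallic-roughness or specular-glossiness converted by MF_SpecGlossToMetalRoughness.
# Named reroutes carry results into MF_GLTFPBR, alongside IOR, alpha mode and attenuation parameters.
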Material /InterchangeAssets/gltf/Substrate/M_GLTF
graph {
 properties {
  bUsedWithNanite=true
  refractionMethod=RM_IndexOfRefraction
  shadingModel=MSM_SubsurfaceProfile
  translucencyLightingMode=TLM_SurfacePerPixelLighting
  usedShadingModels=Subsurface Profile
 }
 nodes [
  n1 = ScalarParameter() {
   controlType=Enumeration
   enumeration=&/EGLTFAlphaMode
   group=BaseColor
   parameterName=AlphaMode
   sortPriority=7
  }
  MaterialFunctionCall!(
   base_color=NamedRerouteUsage()
   metallic=NamedRerouteUsage()
   roughness=NamedRerouteUsage()
   normal=NamedRerouteUsage()
   tangent=NamedRerouteUsage()
   occlusion=NamedRerouteUsage()
   transmission=NamedRerouteUsage()
   specular=NamedRerouteUsage()
   specular_color=NamedRerouteUsage()
   ior=ScalarParameter() {defaultValue=1.5, group=IOR, parameterName=IOR, sliderMax=2, sliderMin=1, sortPriority=0}
   alpha=NamedRerouteUsage()
   alpha_mode=@n1
   alpha_cutoff=ScalarParameter() {group=BaseColor, parameterName=AlphaCutoff, sortPriority=6}
   iridescence=NamedRerouteUsage()
   iridescence_ior=ScalarParameter() {defaultValue=1.3, group=Iridescence, parameterName=IridescenceIOR, sortPriority=0}
   iridescence_thickness=NamedRerouteUsage()
   sheen_color=NamedRerouteUsage()
   sheen_roughness=NamedRerouteUsage()
   clearcoat=NamedRerouteUsage()
   clearcoat_roughness=NamedRerouteUsage()
   clearcoat_normal=NamedRerouteUsage()
   emissive=NamedRerouteUsage()
   emissive_strength=ScalarParameter() {defaultValue=1, group=Emissive, parameterName=EmissiveStrength, sortPriority=2}
   thickness=NamedRerouteUsage()
   attenuation_distance=ScalarParameter() {defaultValue=1, group=Volume, parameterName=AttenuationDistance, sortPriority=0}
   attenuation_color=VectorParameter() {group=Volume, parameterName=AttenuationColor, sortPriority=1, defaultValue=<rgba 1 1 1 0>}.RGB
   anisotropy_strength=NamedRerouteUsage()
   anisotropy_direction=NamedRerouteUsage()
   anisotropy_rotation=ScalarParameter() {group=Anisotropy, parameterName=AnisotropyRotation, sortPriority=2}
   dispersion
  ) {materialFunction=&/MF_GLTFPBR}
  n2 = VectorParameter() {group=BaseColor, parameterName=BaseColorTexture_OffsetScale, sortPriority=2, defaultValue=<rgba 0 0 1 1>}
  n3 = VectorParameter() {group=BaseColor, parameterName=BaseColorTexture_TilingMethod, sortPriority=5, channelNames {a="", b="", g=Tiling V, r=Tiling U}}
  n4 = MaterialFunctionCall!(
   Offset_X=@n2.R
   Offset_Y=@n2.R
   Scale_X=@n2.R
   Scale_Y=@n2.R
   Rotation=ScalarParameter() {group=BaseColor, parameterName=BaseColorTexture_Rotation, sliderMax=6.2832, sortPriority=3}
   TexCoord=ScalarParameter() {group=BaseColor, parameterName=BaseColorTexture_TexCoord, sliderMax=3, sortPriority=4}
   TilingMethod=@n3
  ) {materialFunction=/InterchangeAssets/gltf/MaterialFunctions/MF_TransformUVs}
  n5 = TextureSample!(
   UVs=@n4.UVs
   Tex=TextureObjectParameter() {group=BaseColor, parameterName=BaseColorTexture, sortPriority=0, texture=/InterchangeAssets/gltf/Textures/T_White_srgb}
   "DDX(UVs)"=@n4.UVs
   "DDY(UVs)"=@n4.UVs
   Apply View MipBias
  ) {mipValueMode=TMVM_Derivative, samplerSource=SSM_Clamp_WorldGroupSettings}
  n6 = VectorParameter() {group=BaseColor, parameterName=BaseColorFactor, sortPriority=1, defaultValue=<rgba 1 1 1 1>}
  n7 = Multiply(A=@n5.RGB, B=@n6.RGB)
  n8 = VertexColor()
  NamedRerouteDeclaration(None=StaticSwitch(
    True=Multiply(A=@n7, B=@n8)
    False=@n6.RGB
    Value=StaticBoolParameter() {group=Configuration, parameterName=bHasBaseColorTexture, sortPriority=0}
   )) {name=MRBaseColor, nodeColor=<rgba 0 1 0.5781 1>}
  NamedRerouteDeclaration(None=StaticSwitch(
    True=Multiply(A=@n5.A, B=@n6.A)
    False=@n6.A
    Value=StaticBoolParameter() {group=Configuration, parameterName=bHasBaseColorTexture, sortPriority=0}
   )) {name=MRAlpha, nodeColor=<rgba 1 0.2578 0 1>}
  n9 = VectorParameter() {group=Anisotropy, parameterName=AnisotropyTexture_OffsetScale, sortPriority=3, defaultValue=<rgba 0 0 1 1>}
  n10 = MaterialFunctionCall(
   Offset_X=@n9.R
   Offset_Y=@n9.R
   Scale_X=@n9.R
   Scale_Y=@n9.R
   Rotation=ScalarParameter() {group=Anisotropy, parameterName=AnisotropyTexture_Rotation, sortPriority=4}
   TexCoord=ScalarParameter() {group=Anisotropy, parameterName=AnisotropyTexture_TexCoord, sortPriority=5}
   TilingMethod=VectorParameter() {group=Anisotropy, parameterName=AnisotropyTexture_TilingMethod, sortPriority=6}.RGB
  ) {materialFunction=&MF_TransformUVs}
  n11 = TextureObjectParameter() {
   group=Anisotropy
   parameterName=AnisotropyTexture
   samplerType=SAMPLERTYPE_LinearColor
   sortPriority=0
   texture=/InterchangeAssets/gltf/Textures/T_Anisotropy_Direction_Linear
  }
  n12 = TextureSample(UVs=@n10.UVs, Tex=@n11, "DDX(UVs)"=@n10.UVs, "DDY(UVs)"=@n10.UVs) {mipValueMode=TMVM_Derivative, samplerSource=SSM_Clamp_WorldGroupSettings}
  n13 = ScalarParameter() {group=Anisotropy, parameterName=AnisotropyStrength, sortPriority=1}
  NamedRerouteDeclaration(None=StaticSwitchParameter(True=Multiply(A=@n12.B, B=@n13), False=Constant()) {group=Configuration, parameterName=bHasAnisotropyTexture, sortPriority=6}) {name=Anisotropy, nodeColor=<rgba 0 0.3672 1 1>}
  n14 = ScalarParameter() {defaultValue=100, group=Iridescence, parameterName=IridescenceThicknessMinimum, sortPriority=4}
  n15 = ScalarParameter() {defaultValue=400, group=Iridescence, parameterName=IridescenceThicknessMaximum, sortPriority=5}
  n16 = VectorParameter() {group=Iridescence, parameterName=IridescenceThicknessTexture_OffsetScale, sortPriority=10, defaultValue=<rgba 0 0 1 1>}
  n17 = VectorParameter() {
   group=Iridescence
   parameterName=IridescenceThicknessTexture_TilingMethod
   sortPriority=13
   channelNames {a="", b="", g=Tiling V, r=Tiling U}
  }
  n18 = MaterialFunctionCall(
   Offset_X=@n16.R
   Offset_Y=@n16.R
   Scale_X=@n16.R
   Scale_Y=@n16.R
   Rotation=ScalarParameter() {group=Iridescence, parameterName=IridescenceThicknessTexture_Rotation, sortPriority=11}
   TexCoord=ScalarParameter() {group=Iridescence, parameterName=IridescenceThicknessTexture_TexCoord, sortPriority=12}
   TilingMethod=@n17
  ) {materialFunction=&MF_TransformUVs}
  n19 = TextureObjectParameter() {group=Iridescence, parameterName=IridescenceThicknessTexture, sortPriority=2, texture=&T_White_srgb}
  n20 = TextureSample(UVs=@n18.UVs, Tex=@n19, "DDX(UVs)"=@n18.UVs, "DDY(UVs)"=@n18.UVs) {mipValueMode=TMVM_Derivative, samplerSource=SSM_Clamp_WorldGroupSettings}
  NamedRerouteDeclaration(None=StaticSwitchParameter(True=LinearInterpolate(A=@n14, B=@n15, Alpha=@n20.G), False=Constant()) {group=Configuration, parameterName=bHasIridescence, sortPriority=9}) {name=IridescenceThickness, nodeColor=<rgba 1 0.1641 0 1>}
  n21 = ScalarParameter() {group=Iridescence, parameterName=IridescenceFactor, sortPriority=3}
  n22 = VectorParameter() {group=Iridescence, parameterName=IridescenceTexture_OffsetScale, sortPriority=6, defaultValue=<rgba 0 0 1 1>}
  n23 = VectorParameter() {group=Iridescence, parameterName=IridescenceTexture_TilingMethod, sortPriority=9, channelNames {a="", b="", g=Tiling V, r=Tiling U}}
  n24 = MaterialFunctionCall(
   Offset_X=@n22.R
   Offset_Y=@n22.R
   Scale_X=@n22.R
   Scale_Y=@n22.R
   Rotation=ScalarParameter() {group=Iridescence, parameterName=IridescenceTexture_Rotation, sortPriority=7}
   TexCoord=ScalarParameter() {group=Iridescence, parameterName=IridescenceTexture_TexCoord, sortPriority=8}
   TilingMethod=@n23
  ) {materialFunction=&MF_TransformUVs}
  n25 = TextureObjectParameter() {group=Iridescence, parameterName=IridescenceTexture, sortPriority=1, texture=&T_White_srgb}
  n26 = TextureSample(UVs=@n24.UVs, Tex=@n25, "DDX(UVs)"=@n24.UVs, "DDY(UVs)"=@n24.UVs) {mipValueMode=TMVM_Derivative, samplerSource=SSM_Clamp_WorldGroupSettings}
  NamedRerouteDeclaration(None=StaticSwitchParameter(True=Multiply(A=@n21, B=@n26.R), False=Constant()) {group=Configuration, parameterName=bHasIridescence, sortPriority=9}) {name=Iridescence, nodeColor=<rgba 0 0.1562 1 1>}
  n27 = VectorParameter() {group=Emissive, parameterName=EmissiveTexture_OffsetScale, sortPriority=3, defaultValue=<rgba 0 0 1 1>}
  n28 = VectorParameter() {group=Emissive, parameterName=EmissiveTexture_TilingMethod, sortPriority=6, channelNames {a="", b="", g=Tiling V, r=Tiling U}}
  n29 = MaterialFunctionCall(
   Offset_X=@n27.R
   Offset_Y=@n27.R
   Scale_X=@n27.R
   Scale_Y=@n27.R
   Rotation=ScalarParameter() {group=Emissive, parameterName=EmissiveTexture_Rotation, sliderMax=6.2832, sortPriority=4}
   TexCoord=ScalarParameter() {group=Emissive, parameterName=EmissiveTexture_TexCoord, sliderMax=3, sortPriority=5}
   TilingMethod=@n28
  ) {materialFunction=&MF_TransformUVs}
  n30 = TextureSample(
   UVs=@n29.UVs
   Tex=TextureObjectParameter() {group=Emissive, parameterName=EmissiveTexture, sortPriority=0, texture=&T_White_srgb}
   "DDX(UVs)"=@n29.UVs
   "DDY(UVs)"=@n29.UVs
  ) {mipValueMode=TMVM_Derivative, samplerSource=SSM_Clamp_WorldGroupSettings}
  n31 = VectorParameter() {group=Emissive, parameterName=EmissiveFactor, sortPriority=1, defaultValue=<rgba 0 0 0 1>}
  NamedRerouteDeclaration(None=StaticSwitchParameter(True=Multiply(A=@n30.RGB, B=@n31.RGB), False=@n31.RGB) {group=Configuration, parameterName=bHasEmissiveTexture, sortPriority=2}) {name=Emissive, nodeColor=<rgba 0 1 0.0391 1>}
  n32 = VectorParameter() {group=Sheen, parameterName=SheenColorTexture_OffsetScale, sortPriority=4, defaultValue=<rgba 0 0 1 1>}
  n33 = VectorParameter() {group=Sheen, parameterName=SheenColorTexture_TilingMethod, sortPriority=7, channelNames {a="", b="", g=Tiling V, r=Tiling U}}
  n34 = MaterialFunctionCall(
   Offset_X=@n32.R
   Offset_Y=@n32.R
   Scale_X=@n32.R
   Scale_Y=@n32.R
   Rotation=ScalarParameter() {group=Sheen, parameterName=SheenColorTexture_Rotation, sliderMax=6.2832, sortPriority=5}
   TexCoord=ScalarParameter() {group=Sheen, parameterName=SheenColorTexture_TexCoord, sliderMax=3, sortPriority=6}
   TilingMethod=@n33
  ) {materialFunction=&MF_TransformUVs}
  n35 = TextureSample(
   UVs=@n34.UVs
   Tex=TextureObjectParameter() {group=Sheen, parameterName=SheenColorTexture, sortPriority=0, texture=&T_White_srgb}
   "DDX(UVs)"=@n34.UVs
   "DDY(UVs)"=@n34.UVs
  ) {mipValueMode=TMVM_Derivative, samplerSource=SSM_Clamp_WorldGroupSettings}
  n36 = VectorParameter() {group=Sheen, parameterName=SheenColorFactor, sortPriority=2, defaultValue=<rgba 0 0 0 1>}
  NamedRerouteDeclaration(None=StaticSwitch(
    True=Multiply(A=@n35.RGB, B=@n36.RGB)
    False=@n36.RGB
    Value=StaticBoolParameter() {group=Configuration, parameterName=bHasSheenTexture, sortPriority=7}
   )) {name=SheenColor, nodeColor=<rgba 1 0.3281 0 1>}
  n37 = VectorParameter() {group=Sheen, parameterName=SheenRoughnessTexture_OffsetScale, sortPriority=8, defaultValue=<rgba 0 0 1 1>}
  n38 = VectorParameter() {group=Sheen, parameterName=SheenRoughnessTexture_TilingMethod, sortPriority=11, channelNames {a="", b="", g=Tiling V, r=Tiling U}}
  n39 = MaterialFunctionCall(
   Offset_X=@n37.R
   Offset_Y=@n37.R
   Scale_X=@n37.R
   Scale_Y=@n37.R
   Rotation=ScalarParameter() {group=Sheen, parameterName=SheenRoughnessTexture_Rotation, sliderMax=6.2832, sortPriority=9}
   TexCoord=ScalarParameter() {group=Sheen, parameterName=SheenRoughnessTexture_TexCoord, sliderMax=3, sortPriority=10}
   TilingMethod=@n38
  ) {materialFunction=&MF_TransformUVs}
  n40 = TextureObjectParameter() {
   group=Sheen
   parameterName=SheenRoughnessTexture
   samplerType=SAMPLERTYPE_LinearColor
   sortPriority=1
   texture=/InterchangeAssets/gltf/Textures/T_White_Linear
  }
  n41 = TextureSample(UVs=@n39.UVs, Tex=@n40, "DDX(UVs)"=@n39.UVs, "DDY(UVs)"=@n39.UVs) {mipValueMode=TMVM_Derivative, samplerSource=SSM_Clamp_WorldGroupSettings, samplerType=SAMPLERTYPE_LinearColor}
  n42 = ScalarParameter() {group=Sheen, parameterName=SheenRoughnessFactor, sliderMax=1, sortPriority=3}
  NamedRerouteDeclaration(None=StaticSwitch(
    True=Multiply(A=@n41.A, B=@n42)
    False=@n42
    Value=StaticBoolParameter() {group=Configuration, parameterName=bHasSheenTexture, sortPriority=7}
   )) {name=SheenRoughness, nodeColor=<rgba 0.0078 0 1 1>}
  n43 = VectorParameter() {group=ClearCoat, parameterName=ClearCoatTexture_OffsetScale, sortPriority=6, defaultValue=<rgba 0 0 1 1>}
  n44 = VectorParameter() {group=ClearCoat, parameterName=ClearCoatTexture_TilingMethod, sortPriority=9, channelNames {a="", b="", g=Tiling V, r=Tiling U}}
  n45 = MaterialFunctionCall(
   Offset_X=@n43.R
   Offset_Y=@n43.R
   Scale_X=@n43.R
   Scale_Y=@n43.R
   Rotation=ScalarParameter() {group=ClearCoat, parameterName=ClearCoatTexture_Rotation, sliderMax=6.2832, sortPriority=7}
   TexCoord=ScalarParameter() {group=ClearCoat, parameterName=ClearCoatTexture_TexCoord, sliderMax=3, sortPriority=8}
   TilingMethod=@n44
  ) {materialFunction=&MF_TransformUVs}
  n46 = TextureObjectParameter() {
   group=ClearCoat
   parameterName=ClearCoatTexture
   samplerType=SAMPLERTYPE_LinearColor
   sortPriority=0
   texture=&T_White_Linear
  }
  n47 = TextureSample(UVs=@n45.UVs, Tex=@n46, "DDX(UVs)"=@n45.UVs, "DDY(UVs)"=@n45.UVs) {mipValueMode=TMVM_Derivative, samplerSource=SSM_Clamp_WorldGroupSettings, samplerType=SAMPLERTYPE_LinearColor}
  n48 = ScalarParameter() {group=ClearCoat, parameterName=ClearCoatFactor, sliderMax=1, sortPriority=3}
  NamedRerouteDeclaration(None=StaticSwitch(
    True=Multiply(A=@n47.R, B=@n48)
    False=@n48
    Value=StaticBoolParameter() {group=Configuration, parameterName=bHasClearCoatTexture, sortPriority=6}
   )) {name=ClearCoat, nodeColor=<rgba 0 0.1094 1 1>}
  n49 = VectorParameter() {group=ClearCoat, parameterName=ClearCoatRoughnessTexture_OffsetScale, sortPriority=10, defaultValue=<rgba 0 0 1 1>}
  n50 = VectorParameter() {group=ClearCoat, parameterName=ClearCoatRoughnessTexture_TilingMethod, sortPriority=13, channelNames {a="", b="", g=Tiling V, r=Tiling U}}
  n51 = MaterialFunctionCall(
   Offset_X=@n49.R
   Offset_Y=@n49.R
   Scale_X=@n49.R
   Scale_Y=@n49.R
   Rotation=ScalarParameter() {group=ClearCoat, parameterName=ClearCoatRoughnessTexture_Rotation, sliderMax=6.2832, sortPriority=11}
   TexCoord=ScalarParameter() {group=ClearCoat, parameterName=ClearCoatRoughnessTexture_TexCoord, sliderMax=3, sortPriority=12}
   TilingMethod=@n50
  ) {materialFunction=&MF_TransformUVs}
  n52 = TextureObjectParameter() {
   group=ClearCoat
   parameterName=ClearCoatRoughnessTexture
   samplerType=SAMPLERTYPE_LinearColor
   sortPriority=1
   texture=&T_White_Linear
  }
  n53 = TextureSample(UVs=@n51.UVs, Tex=@n52, "DDX(UVs)"=@n51.UVs, "DDY(UVs)"=@n51.UVs) {mipValueMode=TMVM_Derivative, samplerSource=SSM_Clamp_WorldGroupSettings, samplerType=SAMPLERTYPE_LinearColor}
  n54 = ScalarParameter() {group=ClearCoat, parameterName=ClearCoatRoughnessFactor, sliderMax=1, sortPriority=4}
  NamedRerouteDeclaration(None=StaticSwitch(
    True=Multiply(A=@n53.G, B=@n54)
    False=@n54
    Value=StaticBoolParameter() {group=Configuration, parameterName=bHasClearCoatTexture, sortPriority=6}
   )) {name=ClearCoatRoughness, nodeColor=<rgba 0.4297 1 0 1>}
  n55 = VectorParameter() {group=ClearCoat, parameterName=ClearCoatNormalTexture_OffsetScale, sortPriority=14, defaultValue=<rgba 0 0 1 1>}
  n56 = VectorParameter() {group=ClearCoat, parameterName=ClearCoatNormalTexture_TilingMethod, sortPriority=17, channelNames {a="", b="", g=Tiling V, r=Tiling U}}
  n57 = MaterialFunctionCall(
   Offset_X=@n55.R
   Offset_Y=@n55.R
   Scale_X=@n55.R
   Scale_Y=@n55.R
   Rotation=ScalarParameter() {group=ClearCoat, parameterName=ClearCoatNormalTexture_Rotation, sliderMax=6.2832, sortPriority=15}
   TexCoord=ScalarParameter() {group=ClearCoat, parameterName=ClearCoatNormalTexture_TexCoord, sliderMax=3, sortPriority=16}
   TilingMethod=@n56
  ) {materialFunction=&MF_TransformUVs}
  n58 = TextureObjectParameter() {
   group=ClearCoat
   parameterName=ClearCoatNormalTexture
   samplerType=SAMPLERTYPE_Normal
   sortPriority=2
   texture=/InterchangeAssets/gltf/Textures/T_Generic_N
  }
  n59 = MaterialFunctionCall!(
   NormalsTS=TextureSample(UVs=@n57.UVs, Tex=@n58, "DDX(UVs)"=@n57.UVs, "DDY(UVs)"=@n57.UVs) {mipValueMode=TMVM_Derivative, samplerSource=SSM_Clamp_WorldGroupSettings, samplerType=SAMPLERTYPE_Normal}.RGB
   SinCos=@n57.SinCos
  ) {materialFunction=/InterchangeAssets/gltf/MaterialFunctions/MF_RotateNormals_TS}
  n60 = ScalarParameter() {defaultValue=1, group=ClearCoat, parameterName=ClearCoatNormalScale, sliderMax=1, sortPriority=5}
  NamedRerouteDeclaration(None=StaticSwitch(
    True=MaterialFunctionCall!(Normal=@n59.Result, Scale=@n60) {materialFunction=/InterchangeAssets/gltf/MaterialFunctions/MF_ScaleNormal}.Result
    False=Constant3Vector() {constant=<rgba 0 0 1 0>}
    Value=StaticBoolParameter() {group=Configuration, parameterName=bHasClearCoatTexture, sortPriority=6}
   )) {name=ClearCoatNormal, nodeColor=<rgba 1 0 0.0469 1>}
  MaterialExpressionVertexTangentWS_2 = VertexTangentWS()
  n61 = Custom!(SurfaceTangent=@MaterialExpressionVertexTangentWS_2) {
   code="{\r\n##if MATERIAL_TANGENTSPACENORMAL\r\n\tconst float3 T = float3(1,0,0);\r\n##else\r\n\tconst float3 T = normalize(SurfaceTangent);\r\n##endif\r\n\treturn T;\r\n}"
   description=GetMtlDefTangentFrame
   inputs [{inputName=SurfaceTangent, input {expression=&self:MaterialExpressionVertexTangentWS_2}}]
  }
  NamedRerouteDeclaration(None=@n61) {name=Tangent, nodeColor=<rgba 1 0 0.5391 1>}
  n62 = ScalarParameter() {defaultValue=1, group=MetallicRoughness, parameterName=RoughnessFactor, sliderMax=1, sortPriority=2}
  n63 = VectorParameter() {group=MetallicRoughness, parameterName=MetallicRoughnessTexture_OffsetScale, sortPriority=3, defaultValue=<rgba 0 0 1 1>}
  n64 = VectorParameter() {
   group=MetallicRoughness
   parameterName=MetallicRoughnessTexture_TilingMethod
   sortPriority=6
   channelNames {a="", b="", g=Tiling V, r=Tiling U}
  }
  n65 = MaterialFunctionCall(
   Offset_X=@n63.R
   Offset_Y=@n63.R
   Scale_X=@n63.R
   Scale_Y=@n63.R
   Rotation=ScalarParameter() {group=MetallicRoughness, parameterName=MetallicRoughnessTexture_Rotation, sliderMax=6.2832, sortPriority=4}
   TexCoord=ScalarParameter() {group=MetallicRoughness, parameterName=MetallicRoughnessTexture_TexCoord, sliderMax=3, sortPriority=5}
   TilingMethod=@n64
  ) {materialFunction=&MF_TransformUVs}
  n66 = TextureObjectParameter() {
   group=MetallicRoughness
   parameterName=MetallicRoughnessTexture
   samplerType=SAMPLERTYPE_LinearColor
   sortPriority=0
   texture=&T_White_Linear
  }
  n67 = TextureSample(UVs=@n65.UVs, Tex=@n66, "DDX(UVs)"=@n65.UVs, "DDY(UVs)"=@n65.UVs) {mipValueMode=TMVM_Derivative, samplerSource=SSM_Clamp_WorldGroupSettings, samplerType=SAMPLERTYPE_LinearColor}
  NamedRerouteDeclaration(None=StaticSwitch(
    True=Multiply(A=@n62, B=@n67.G)
    False=@n62
    Value=StaticBoolParameter() {group=Configuration, parameterName=bHasMetallicRoughnessTexture, sortPriority=1}
   )) {name=MRRoughness, nodeColor=<rgba 1 0 0.7031 1>}
  n68 = VectorParameter() {group=DiffuseSpecGloss, parameterName=DiffuseTexture_OffsetScale, sortPriority=5, defaultValue=<rgba 0 0 1 1>}
  n69 = VectorParameter() {group=DiffuseSpecGloss, parameterName=DiffuseTexture_TilingMethod, sortPriority=8, channelNames {a="", b="", g=Tiling V, r=Tiling U}}
  n70 = MaterialFunctionCall(
   Offset_X=@n68.R
   Offset_Y=@n68.R
   Scale_X=@n68.R
   Scale_Y=@n68.R
   Rotation=ScalarParameter() {group=DiffuseSpecGloss, parameterName=DiffuseTexture_Rotation, sliderMax=6.2832, sortPriority=6}
   TexCoord=ScalarParameter() {group=DiffuseSpecGloss, parameterName=DiffuseTexture_TexCoord, sliderMax=3, sortPriority=7}
   TilingMethod=@n69
  ) {materialFunction=&MF_TransformUVs}
  n71 = TextureObjectParameter() {group=DiffuseSpecGloss, parameterName=DiffuseTexture, sortPriority=0, texture=&T_White_srgb}
  n72 = TextureSample(UVs=@n70.UVs, Tex=@n71, "DDX(UVs)"=@n70.UVs, "DDY(UVs)"=@n70.UVs) {mipValueMode=TMVM_Derivative, samplerSource=SSM_Clamp_WorldGroupSettings}
  n73 = VectorParameter() {group=DiffuseSpecGloss, parameterName=DiffuseFactor, sortPriority=2, defaultValue=<rgba 1 1 1 1>}
  n74 = StaticBoolParameter() {group=Configuration, parameterName=bHasDiffuseSpecGlossTexture, sortPriority=0}
  NamedRerouteDeclaration(None=StaticSwitch(True=Multiply(A=@n72.RGB, B=@n73.RGB), False=@n73.RGB, Value=@n74)) {name=SpecGlossDiffuseColor, nodeColor=<rgba 0 1 0.5078 1>}
  NamedRerouteDeclaration(None=StaticSwitch(True=Multiply(A=@n72.A, B=@n73.A), False=@n73.A, Value=@n74)) {name=SpecGlossAlpha, nodeColor=<rgba 1 0.1875 0 1>}
  n75 = VectorParameter() {group=Volume, parameterName=ThicknessTexture_OffsetScale, sortPriority=6, defaultValue=<rgba 0 0 1 1>}
  n76 = VectorParameter() {group=Volume, parameterName=ThicknessTexture_TilingMethod, sortPriority=9, channelNames {a="", b="", g=Tiling V, r=Tiling U}}
  n77 = MaterialFunctionCall(
   Offset_X=@n75.R
   Offset_Y=@n75.R
   Scale_X=@n75.R
   Scale_Y=@n75.R
   Rotation=ScalarParameter() {group=Volume, parameterName=ThicknessTexture_Rotation, sliderMax=6.2832, sortPriority=7}
   TexCoord=ScalarParameter() {group=Volume, parameterName=ThicknessTexture_TexCoord, sliderMax=3, sortPriority=8}
   TilingMethod=@n76
  ) {materialFunction=&MF_TransformUVs}
  n78 = TextureObjectParameter() {
   group=Volume
   parameterName=ThicknessTexture
   samplerType=SAMPLERTYPE_LinearColor
   sortPriority=3
   texture=&T_White_Linear
  }
  n79 = TextureSample(UVs=@n77.UVs, Tex=@n78, "DDX(UVs)"=@n77.UVs, "DDY(UVs)"=@n77.UVs) {mipValueMode=TMVM_Derivative, samplerSource=SSM_Clamp_WorldGroupSettings, samplerType=SAMPLERTYPE_LinearColor}
  n80 = ScalarParameter() {group=Volume, parameterName=ThicknessFactor, sortPriority=2}
  NamedRerouteDeclaration(None=StaticSwitch(
    True=Multiply(A=@n79.G, B=@n80)
    False=@n80
    Value=StaticBoolParameter() {group=Configuration, parameterName=bHasThicknessTexture, sortPriority=6}
   )) {name=Thickness, nodeColor=<rgba 0.3359 1 0 1>}
  n81 = VectorParameter() {group=DiffuseSpecGloss, parameterName=SpecularGlossinessTexture_OffsetScale, sortPriority=9, defaultValue=<rgba 0 0 1 1>}
  n82 = VectorParameter() {
   group=DiffuseSpecGloss
   parameterName=SpecularGlossinessTexture_TilingMethod
   sortPriority=12
   channelNames {a="", b="", g=Tiling V, r=Tiling U}
  }
  n83 = MaterialFunctionCall(
   Offset_X=@n81.R
   Offset_Y=@n81.R
   Scale_X=@n81.R
   Scale_Y=@n81.R
   Rotation=ScalarParameter() {group=DiffuseSpecGloss, parameterName=SpecularGlossinessTexture_Rotation, sliderMax=6.2832, sortPriority=10}
   TexCoord=ScalarParameter() {group=DiffuseSpecGloss, parameterName=SpecularGlossinessTexture_TexCoord, sliderMax=3, sortPriority=11}
   TilingMethod=@n82
  ) {materialFunction=&MF_TransformUVs}
  n84 = TextureObjectParameter() {group=DiffuseSpecGloss, parameterName=SpecularGlossinessTexture, sortPriority=1, texture=&T_White_srgb}
  n85 = TextureSample(UVs=@n83.UVs, Tex=@n84, "DDX(UVs)"=@n83.UVs, "DDY(UVs)"=@n83.UVs) {mipValueMode=TMVM_Derivative, samplerSource=SSM_Clamp_WorldGroupSettings}
  n86 = VectorParameter() {group=DiffuseSpecGloss, parameterName=SpecFactor, sortPriority=3, defaultValue=<rgba 1 1 1 0>}
  NamedRerouteDeclaration(None=StaticSwitch(
    True=Multiply(A=@n85.RGB, B=@n86.RGB)
    False=@n86.RGB
    Value=StaticBoolParameter() {group=Configuration, parameterName=bHasDiffuseSpecGlossTexture, sortPriority=0}
   )) {name=SpecGlossSpecularColor, nodeColor=<rgba 0.4531 1 0 1>}
  n87 = MaterialFunctionCall!(DiffuseColor=NamedRerouteUsage(), SpecularColor=NamedRerouteUsage(), Glossiness=NamedRerouteUsage()) {materialFunction=/InterchangeAssets/gltf/MaterialFunctions/MF_SpecGlossToMetalRoughness}
  NamedRerouteDeclaration(None=StaticSwitchParameter(True=@n87."BaseColor (V3)", False=NamedRerouteUsage()) {group=Configuration, parameterName=bHasDiffuseSpecGloss, sortPriority=0}) {name=BaseColor, nodeColor=<rgba 1 0 0.7734 1>}
  NamedRerouteDeclaration(None=StaticSwitchParameter(True=NamedRerouteUsage(), False=NamedRerouteUsage()) {group=Configuration, parameterName=bHasDiffuseSpecGloss, sortPriority=0}) {name=Alpha, nodeColor=<rgba 0 1 0.9063 1>}
  NamedRerouteDeclaration(None=StaticSwitchParameter(True=@n87.Roughness, False=NamedRerouteUsage()) {group=Configuration, parameterName=bHasDiffuseSpecGloss, sortPriority=0}) {name=Roughness, nodeColor=<rgba 1 0.5859 0 1>}
  n88 = ScalarParameter() {defaultValue=1, group=MetallicRoughness, parameterName=MetallicFactor, sliderMax=1, sortPriority=1}
  NamedRerouteDeclaration(None=StaticSwitch(
    True=Multiply(A=@n67.B, B=@n88)
    False=@n88
    Value=StaticBoolParameter() {group=Configuration, parameterName=bHasMetallicRoughnessTexture, sortPriority=1}
   )) {name=MRMetal, nodeColor=<rgba 0 1 0.9766 1>}
  n89 = ScalarParameter() {defaultValue=1, group=DiffuseSpecGloss, parameterName=GlossinessFactor, sliderMax=1, sortPriority=4}
  NamedRerouteDeclaration(None=StaticSwitch(
    True=Multiply(A=@n85.A, B=@n89)
    False=@n89
    Value=StaticBoolParameter() {group=Configuration, parameterName=bHasDiffuseSpecGlossTexture, sortPriority=0}
   )) {name=Glossiness, nodeColor=<rgba 1 0 0.7969 1>}
  NamedRerouteDeclaration(None=StaticSwitchParameter(True=@n87."Metallic (S)", False=NamedRerouteUsage()) {group=Configuration, parameterName=bHasDiffuseSpecGloss, sortPriority=0}) {name=Metal, nodeColor=<rgba 0 1 0.8828 1>}
  n90 = VectorParameter() {group=Normals, parameterName=NormalTexture_OffsetScale, sortPriority=2, defaultValue=<rgba 0 0 1 1>}
  n91 = VectorParameter() {group=Normals, parameterName=NormalTexture_TilingMethod, sortPriority=5, channelNames {a="", b="", g=Tiling V, r=Tiling U}}
  n92 = MaterialFunctionCall(
   Offset_X=@n90.R
   Offset_Y=@n90.R
   Scale_X=@n90.R
   Scale_Y=@n90.R
   Rotation=ScalarParameter() {group=Normals, parameterName=NormalTexture_Rotation, sliderMax=6.2832, sortPriority=3}
   TexCoord=ScalarParameter() {group=Normals, parameterName=NormalTexture_TexCoord, sliderMax=3, sortPriority=4}
   TilingMethod=@n91
  ) {materialFunction=&MF_TransformUVs}
  n93 = TextureObjectParameter() {
   group=Normals
   parameterName=NormalTexture
   samplerType=SAMPLERTYPE_Normal
   sortPriority=0
   texture=&T_Generic_N
  }
  n94 = MaterialFunctionCall(
   NormalsTS=TextureSample(UVs=@n92.UVs, Tex=@n93, "DDX(UVs)"=@n92.UVs, "DDY(UVs)"=@n92.UVs) {mipValueMode=TMVM_Derivative, samplerSource=SSM_Clamp_WorldGroupSettings, samplerType=SAMPLERTYPE_Normal}.RGB
   SinCos=@n92.SinCos
  ) {materialFunction=&MF_RotateNormals_TS}
  n95 = ScalarParameter() {defaultValue=1, group=Normals, parameterName=NormalScale, sortPriority=1}
  NamedRerouteDeclaration(None=StaticSwitchParameter(
    True=MaterialFunctionCall(Normal=@n94.Result, Scale=@n95) {materialFunction=&MF_ScaleNormal}.Result
    False=Constant3Vector() {constant=<rgba 0 0 1 0>}
   ) {group=Configuration, parameterName=bHasNormalTexture, sortPriority=3}) {name=Normal, nodeColor=<rgba 0.5 0.5 1 1>}
  n96 = VectorParameter() {group=Occlusion, parameterName=OcclusionTexture_OffsetScale, sortPriority=2, defaultValue=<rgba 0 0 1 1>}
  n97 = VectorParameter() {group=Occlusion, parameterName=OcclusionTexture_TilingMethod, sortPriority=5, channelNames {a="", b="", g=Tiling V, r=Tiling U}}
  n98 = MaterialFunctionCall(
   Offset_X=@n96.R
   Offset_Y=@n96.R
   Scale_X=@n96.R
   Scale_Y=@n96.R
   Rotation=ScalarParameter() {group=Occlusion, parameterName=OcclusionTexture_Rotation, sliderMax=6.2832, sortPriority=3}
   TexCoord=ScalarParameter() {group=Occlusion, parameterName=OcclusionTexture_TexCoord, sliderMax=3, sortPriority=4}
   TilingMethod=@n97
  ) {materialFunction=&MF_TransformUVs}
  n99 = TextureObjectParameter() {
   group=Occlusion
   parameterName=OcclusionTexture
   samplerType=SAMPLERTYPE_LinearColor
   sortPriority=0
   texture=&T_White_Linear
  }
  n100 = TextureSample(UVs=@n98.UVs, Tex=@n99, "DDX(UVs)"=@n98.UVs, "DDY(UVs)"=@n98.UVs) {mipValueMode=TMVM_Derivative, samplerSource=SSM_Clamp_WorldGroupSettings, samplerType=SAMPLERTYPE_LinearColor}
  n101 = ScalarParameter() {defaultValue=1, group=Occlusion, parameterName=OcclusionStrength, sliderMax=1, sortPriority=1}
  NamedRerouteDeclaration(None=StaticSwitchParameter(True=LinearInterpolate(B=@n100.R, Alpha=@n101) {constA=1}, False=@n101) {group=Configuration, parameterName=bHasOcclusionTexture, sortPriority=4}) {name=Occlusion, nodeColor=<rgba 0 0.7187 1 1>}
  n102 = VectorParameter() {group=Specular, parameterName=SpecularTexture_OffsetScale, sortPriority=4, defaultValue=<rgba 0 0 1 1>}
  n103 = VectorParameter() {group=Specular, parameterName=SpecularTexture_TilingMethod, sortPriority=7, channelNames {a="", b="", g=Tiling V, r=Tiling U}}
  n104 = MaterialFunctionCall(
   Offset_X=@n102.R
   Offset_Y=@n102.R
   Scale_X=@n102.R
   Scale_Y=@n102.R
   Rotation=ScalarParameter() {group=Specular, parameterName=SpecularTexture_Rotation, sliderMax=6.2832, sortPriority=5}
   TexCoord=ScalarParameter() {group=Specular, parameterName=SpecularTexture_TexCoord, sliderMax=3, sortPriority=6}
   TilingMethod=@n103
  ) {materialFunction=&MF_TransformUVs}
  n105 = TextureObjectParameter() {
   group=Specular
   parameterName=SpecularTexture
   samplerType=SAMPLERTYPE_LinearColor
   sortPriority=0
   texture=&T_White_Linear
  }
  n106 = TextureSample(UVs=@n104.UVs, Tex=@n105, "DDX(UVs)"=@n104.UVs, "DDY(UVs)"=@n104.UVs) {mipValueMode=TMVM_Derivative, samplerSource=SSM_Clamp_WorldGroupSettings}
  n107 = ScalarParameter() {defaultValue=1, group=Specular, parameterName=SpecularFactor, sliderMax=1, sortPriority=1}
  NamedRerouteDeclaration(None=StaticSwitchParameter(True=Multiply(A=@n106.A, B=@n107), False=@n107) {group=Configuration, parameterName=bHasSpecularTexture, sortPriority=5}) {name=Specular, nodeColor=<rgba 1 0 0.7266 1>}
  n108 = VectorParameter() {group=Specular, parameterName=SpecularColorTexture_OffsetScale, sortPriority=8, defaultValue=<rgba 0 0 1 1>}
  n109 = VectorParameter() {group=Specular, parameterName=SpecularColorTexture_TilingMethod, sortPriority=11, channelNames {a="", b="", g=Tiling V, r=Tiling U}}
  n110 = MaterialFunctionCall(
   Offset_X=@n108.R
   Offset_Y=@n108.R
   Scale_X=@n108.R
   Scale_Y=@n108.R
   Rotation=ScalarParameter() {group=Specular, parameterName=SpecularColorTexture_Rotation, sliderMax=6.2832, sortPriority=9}
   TexCoord=ScalarParameter() {group=Specular, parameterName=SpecularColorTexture_TexCoord, sliderMax=3, sortPriority=10}
   TilingMethod=@n109
  ) {materialFunction=&MF_TransformUVs}
  n111 = TextureSample(
   UVs=@n110.UVs
   Tex=TextureObjectParameter() {group=Specular, parameterName=SpecularColorTexture, sortPriority=2, texture=&T_White_srgb}
   "DDX(UVs)"=@n110.UVs
   "DDY(UVs)"=@n110.UVs
  ) {mipValueMode=TMVM_Derivative, samplerSource=SSM_Clamp_WorldGroupSettings}
  n112 = VectorParameter() {group=Specular, parameterName=SpecularColorFactor, sortPriority=3, defaultValue=<rgba 1 1 1 1>}
  NamedRerouteDeclaration(None=StaticSwitchParameter(True=Multiply(A=@n111.RGB, B=@n112.RGB), False=@n112.RGB) {group=Configuration, parameterName=bHasSpecularColorTexture, sortPriority=5}) {name=SpecularColor, nodeColor=<rgba 0 1 0.9531 1>}
  n113 = VectorParameter() {group=Transmission, parameterName=TransmissionTexture_OffsetScale, sortPriority=2, defaultValue=<rgba 0 0 1 1>}
  n114 = VectorParameter() {group=Transmission, parameterName=TransmissionTexture_TilingMethod, sortPriority=5, channelNames {a="", b="", g=Tiling V, r=Tiling U}}
  n115 = MaterialFunctionCall(
   Offset_X=@n113.R
   Offset_Y=@n113.R
   Scale_X=@n113.R
   Scale_Y=@n113.R
   Rotation=ScalarParameter() {group=Transmission, parameterName=TransmissionTexture_Rotation, sliderMax=6.2832, sortPriority=3}
   TexCoord=ScalarParameter() {group=Transmission, parameterName=TransmissionTexture_TexCoord, sliderMax=3, sortPriority=4}
   TilingMethod=@n114
  ) {materialFunction=&MF_TransformUVs}
  n116 = TextureObjectParameter() {
   group=Transmission
   parameterName=TransmissionTexture
   samplerType=SAMPLERTYPE_LinearColor
   sortPriority=0
   texture=&T_White_Linear
  }
  n117 = TextureSample(UVs=@n115.UVs, Tex=@n116, "DDX(UVs)"=@n115.UVs, "DDY(UVs)"=@n115.UVs) {mipValueMode=TMVM_Derivative, samplerSource=SSM_Clamp_WorldGroupSettings, samplerType=SAMPLERTYPE_LinearColor}
  n118 = ScalarParameter() {group=Transmission, parameterName=TransmissionFactor, sliderMax=1, sortPriority=1}
  NamedRerouteDeclaration(None=StaticSwitchParameter(True=Multiply(A=@n117.R, B=@n118), False=@n118) {group=Configuration, parameterName=bHasTransmissionTexture, sortPriority=8}) {name=Transmission, nodeColor=<rgba 0 0.6484 1 1>}
  n119 = AppendVector(A=@n12.R, B=@n12.R)
  NamedRerouteDeclaration(None=StaticSwitchParameter(True=ConstantBiasScale(None=@n119) {bias=-0.5, scale=2}, False=Constant2Vector() {r=1}) {group=Configuration, parameterName=bHasAnisotropyTexture, sortPriority=6}) {name=AnisotropyDirection, nodeColor=<rgba 0 1 0.2031 1>}
 ]
}
